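; Standard vehicle boost cosmetic actor that plays a Niagara effect and audio behavior while boosting.
; Boost start activates both if the significance LOD check passes; boost end and EndPlay stop them.
; On apply, Niagara auto-manages attachment to the parent actor's root at the attach socket.
; Color uses custom parameters on Niagara, otherwise a painted DataTable row tints particles and chassis port, with a default port color for missing rows.
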(event Vehicle|Cosmetics|EventOnBoostStarted
  (bind _returnvalue (VehicleCosmetics|FX|GetMaxSignificanceLODBoostFX))
  (bind _btrigger (Class|BPVehicleCosmeticsLibrary|ShouldTriggerFX (Variables|Vehicle|Cosmetics|GetVehicleCosmeticsInterface) _returnvalue))
  (if _btrigger
    (Components|Activation|SetActive (Variables|Default|GetNiagara) true false)
    (AudioGameplayBehavior|Start (Variables|Default|GetAGB_Vehicle_Boost_Standard))))

(event Vehicle|Cosmetics|EventOnCosmeticApplied
  (bind _returnvalue (Actor|GetAttachParentActor))
  (bind _niagara (Variables|Default|GetNiagara))
  (bind _rootcomponent (Variables|/Script/Engine.Actor|GetRootComponent _returnvalue))
  (Utilities|IsValid _returnvalue
    (:Is%20Valid
      (Utilities|IsValid _rootcomponent
        (:Is%20Valid
          (Effects|Components|ParticleSystem|SetAutoAttachmentParameters _niagara _rootcomponent (Actor|GetAttachParentSocketName) "SnapToTarget" "SnapToTarget" "SnapToTarget")
          (Effects|Components|ParticleSystem|SetUseAutoManageAttachment _niagara true))
        (:Is%20Not%20Valid
          _)))
    (:Is%20Not%20Valid
      _))
  (Development|PrintString "Couldnt setup AutoManageAttachment" true true "(R=1.000000,G=0.848041,B=0.000000,A=1.000000)" 5.0))

(event Vehicle|Cosmetics|EventOnBoostFinished
  (Components|Activation|SetActive (Variables|Default|GetNiagara) false false)
  (AudioGameplayBehavior|Stop (Variables|Default|GetAGB_Vehicle_Boost_Standard)))

(event EventEndPlay
  (params (EndPlayReason :type EEndPlayReason%20Enum :default "Destroyed"))
  (Components|Activation|SetActive (Variables|Default|GetNiagara) false false)
  (AudioGameplayBehavior|Stop (Variables|Default|GetAGB_Vehicle_Boost_Standard)))

(event Vehicle|Cosmetics|EventOnColorReady
  (params (Color :type Name :default "None") (AdditionalInfo :type Vehicle%20Cosmetics%20Additional%20Parameter%20Infos%20Structure))
  (bind (_floatparameters _intparameters _colorparameters) (Utilities|Struct|BreakVehicleCosmeticsAdditionalParameterInfos AdditionalInfo))
  (bind _returnvalue (Vehicle|Cosmetics|GetBodyMeshComponent))
  (if (> (Utilities|Array|Length _colorparameters) 0)
    (for _array_element _colorparameters
      (Effects|Components|ParticleSystem|SetColorParameter (Variables|Default|GetNiagara) (Utilities|Struct|BreakVehicleCosmeticsCustomizableObjectColorInfo _array_element) (.ParameterValue (Utilities|Struct|BreakVehicleCosmeticsCustomizableObjectColorInfo _array_element))))
    (else
      (bind _returnvalue_2 (Utilities|GetDataTableRow "/VehicleCosmetics/DataTables/DT_VehicleCosmetics_Painted.DT_VehicleCosmetics_Painted" Color)
        (:then
          (Effects|Components|ParticleSystem|SetColorParameter (Variables|Default|GetNiagara) "Particles_Color" (Utilities|Struct|BreakStructVehicleCosmeticsPainted _returnvalue_2))
          ;[1 1 1] Paint Chassis Port
          (Rendering|Material|SetVectorParameterValueonMaterials _returnvalue "Chassis_BoostColor" (Math|Conversions|ToVector(LinearColor) (Utilities|Struct|BreakStructVehicleCosmeticsPainted _returnvalue_2))))
        (:RowNotFound
          (Rendering|Material|SetVectorParameterValueonMaterials _returnvalue "Chassis_BoostColor" (Math|Conversions|ToVector(LinearColor) (Variables|Default|GetDefaultPortColor))))))))

(fn ConstructionScript)
